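# Material function that fakes infinite depth with stacked parallax layers of one input texture, named as part of a dealership shared library.
# Tiled, panned UVs feed a chain of BumpOffsets, each driving a texture sample with optional rotation.
# Samples are summed, optionally weighted by a halving gradient falloff.
# A quality switch selects the output, with 16 layers at Epic, 8 at High and Medium, and 4 at Low.
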MaterialFunction /DealershipSharedLibrary/Materials/MaterialFunctions/MF_DLR_InfinityOffset
graph {
 properties {bExposeToLibrary=true}
 nodes [
  Constant() {r=0.05}
  FunctionOutput(None=QualitySwitch(Default=NamedRerouteUsage(), Low=NamedRerouteUsage(), High=NamedRerouteUsage(), Medium=NamedRerouteUsage(), Epic=NamedRerouteUsage())) {outputName=Quality Infinite Offset}
  NamedRerouteDeclaration(None=FunctionInput() {inputName=Texture, inputType=FunctionInput_Texture2D}) {name=InTexture, nodeColor=<rgba 0 1 0.3438 1>}
  n1 = Constant() {r=1}
  n2 = Constant() {desc=(3/8), r=0.375}
  NamedRerouteDeclaration(None=Multiply(A=QualitySwitch(Default=@n1, Low=@n1, High=@n1, Medium=@n2, Epic=@n1), B=NamedRerouteUsage())) {name=Quality_11, nodeColor=<rgba 0.6172 1 0 1>}
  n3 = Constant() {r=1}
  n4 = Constant() {desc=(2/8), r=0.25}
  NamedRerouteDeclaration(None=Multiply(A=QualitySwitch(Default=@n3, Low=@n3, High=@n3, Medium=@n4, Epic=@n3), B=NamedRerouteUsage())) {name=Quality_10, nodeColor=<rgba 0.6172 1 0 1>}
  n5 = Constant() {r=1}
  n6 = Constant() {desc=(1/8), r=0.125}
  NamedRerouteDeclaration(None=Multiply(A=QualitySwitch(Default=@n5, Low=@n5, High=@n5, Medium=@n6, Epic=@n5), B=NamedRerouteUsage())) {name=Quality_09, nodeColor=<rgba 0.6172 1 0 1>}
  n7 = TextureCoordinate()
  n8 = FunctionInput() {bUsePreviewValueAsDefault=true, inputName=UVs, inputType=FunctionInput_Vector4, previewValue=<1 1 0 0>}
  n9 = ComponentMask(None=@n8) {g=true, r=true}
  n10 = BumpOffset(Coordinate=Panner(Coordinate=Multiply(A=@n7, B=@n9), Speed=ComponentMask(None=@n8) {a=true, b=true}), Height=NamedRerouteUsage())
  n11 = NamedRerouteUsage()
  n12 = TextureSample(UVs=StaticSwitch(True=Rotator(Coordinate=@n10), False=@n10, Value=NamedRerouteUsage()), Tex=@n11)
  n13 = Multiply(A=Multiply(B=NamedRerouteUsage()) {constA=327.7}, B=@n12.RGB)
  n14 = NamedRerouteUsage()
  n15 = BumpOffset(Coordinate=@n10, Height=NamedRerouteUsage())
  n16 = TextureSample(UVs=StaticSwitch(True=Rotator(), False=@n15, Value=NamedRerouteUsage()), Tex=@n11)
  n17 = Multiply(A=Multiply(B=NamedRerouteUsage()) {constA=163.8}, B=@n16.RGB)
  n18 = NamedRerouteUsage()
  n19 = BumpOffset(Coordinate=@n15, Height=NamedRerouteUsage())
  n20 = NamedRerouteUsage()
  n21 = TextureSample(UVs=StaticSwitch(True=Rotator(Coordinate=@n19), False=@n19, Value=NamedRerouteUsage()), Tex=@n20)
  n22 = Multiply(A=Multiply(B=NamedRerouteUsage()) {constA=81.92}, B=@n21.RGB)
  n23 = NamedRerouteUsage()
  n24 = BumpOffset(Coordinate=@n19, Height=NamedRerouteUsage())
  n25 = TextureSample(UVs=StaticSwitch(True=Rotator(), False=@n24, Value=NamedRerouteUsage()), Tex=@n20)
  n26 = Multiply(A=Multiply(B=NamedRerouteUsage()) {constA=40.96}, B=@n25.RGB)
  n27 = NamedRerouteUsage()
  n28 = Add(
   A=Add(A=StaticSwitch(True=@n13, False=@n12.RGB, Value=@n14), B=StaticSwitch(True=@n17, False=@n16.RGB, Value=@n18))
   B=Add(A=StaticSwitch(True=@n22, False=@n21.RGB, Value=@n23), B=StaticSwitch(True=@n26, False=@n25.RGB, Value=@n27))
  )
  n29 = NamedRerouteUsage()
  n30 = BumpOffset(Coordinate=@n24, Height=NamedRerouteUsage())
  n31 = NamedRerouteUsage()
  n32 = TextureSample(UVs=StaticSwitch(True=Rotator(Coordinate=@n30), False=@n30, Value=NamedRerouteUsage()), Tex=@n31)
  n33 = StaticSwitch(True=Multiply(A=Multiply(B=@n29) {constA=20.48}, B=@n32.RGB), False=@n32.RGB, Value=NamedRerouteUsage())
  n34 = NamedRerouteUsage()
  n35 = BumpOffset(Coordinate=@n30, Height=NamedRerouteUsage())
  n36 = TextureSample(UVs=StaticSwitch(True=Rotator(), False=@n35, Value=NamedRerouteUsage()), Tex=@n31)
  n37 = StaticSwitch(True=Multiply(A=Multiply(B=@n34) {constA=10.24}, B=@n36.RGB), False=@n36.RGB, Value=NamedRerouteUsage())
  n38 = NamedRerouteUsage()
  n39 = BumpOffset(Coordinate=@n35, Height=NamedRerouteUsage())
  n40 = NamedRerouteUsage()
  n41 = TextureSample(UVs=StaticSwitch(True=Rotator(Coordinate=@n39), False=@n39, Value=NamedRerouteUsage()), Tex=@n40)
  n42 = StaticSwitch(True=Multiply(A=Multiply(B=@n38) {constA=5.12}, B=@n41.RGB), False=@n41.RGB, Value=NamedRerouteUsage())
  n43 = NamedRerouteUsage()
  n44 = BumpOffset(Coordinate=@n39, Height=NamedRerouteUsage())
  n45 = TextureSample(UVs=StaticSwitch(True=Rotator(), False=@n44, Value=NamedRerouteUsage()), Tex=@n40)
  n46 = StaticSwitch(True=Multiply(A=Multiply(B=@n43) {constA=2.56}, B=@n45.RGB), False=@n45.RGB, Value=NamedRerouteUsage())
  n47 = Add(A=@n28, B=Add(A=Add(A=@n33, B=@n37), B=Add(A=@n42, B=@n46)))
  n48 = BumpOffset(Coordinate=@n44, Height=NamedRerouteUsage())
  n49 = NamedRerouteUsage()
  n50 = TextureSample(UVs=StaticSwitch(True=Rotator(Coordinate=@n48), False=@n48, Value=NamedRerouteUsage()), Tex=@n49)
  n51 = Multiply(A=Multiply(B=NamedRerouteUsage()) {constA=1.28}, B=@n50.RGB)
  n52 = NamedRerouteUsage()
  n53 = BumpOffset(Coordinate=@n48, Height=NamedRerouteUsage())
  n54 = TextureSample(UVs=StaticSwitch(True=Rotator(), False=@n53, Value=NamedRerouteUsage()), Tex=@n49)
  n55 = Multiply(A=Multiply(B=NamedRerouteUsage()) {constA=0.64}, B=@n54.RGB)
  n56 = NamedRerouteUsage()
  n57 = BumpOffset(Coordinate=@n53, Height=NamedRerouteUsage())
  n58 = NamedRerouteUsage()
  n59 = TextureSample(UVs=StaticSwitch(True=Rotator(Coordinate=@n57), False=@n57, Value=NamedRerouteUsage()), Tex=@n58)
  n60 = Multiply(A=Multiply(B=NamedRerouteUsage()) {constA=0.32}, B=@n59.RGB)
  n61 = NamedRerouteUsage()
  n62 = BumpOffset(Coordinate=@n57, Height=NamedRerouteUsage())
  n63 = TextureSample(UVs=StaticSwitch(True=Rotator(), False=@n62, Value=NamedRerouteUsage()), Tex=@n58)
  n64 = Multiply(A=Multiply(B=NamedRerouteUsage()) {constA=0.16}, B=@n63.RGB)
  n65 = NamedRerouteUsage()
  n66 = Add(
   A=Add(A=StaticSwitch(True=@n51, False=@n50.RGB, Value=@n52), B=StaticSwitch(True=@n55, False=@n54.RGB, Value=@n56))
   B=Add(A=StaticSwitch(True=@n60, False=@n59.RGB, Value=@n61), B=StaticSwitch(True=@n64, False=@n63.RGB, Value=@n65))
  )
  n67 = BumpOffset(Coordinate=@n62, Height=NamedRerouteUsage())
  n68 = NamedRerouteUsage()
  n69 = TextureSample(UVs=StaticSwitch(True=Rotator(Coordinate=@n67), False=@n67, Value=NamedRerouteUsage()), Tex=@n68)
  n70 = Multiply(A=Multiply(B=NamedRerouteUsage()) {constA=0.08}, B=@n69.RGB)
  n71 = NamedRerouteUsage()
  n72 = BumpOffset(Coordinate=@n67, Height=NamedRerouteUsage())
  n73 = TextureSample(UVs=StaticSwitch(True=Rotator(), False=@n72, Value=NamedRerouteUsage()), Tex=@n68)
  n74 = Multiply(A=Multiply(B=NamedRerouteUsage()) {constA=0.04}, B=@n73.RGB)
  n75 = NamedRerouteUsage()
  n76 = BumpOffset(Coordinate=@n72, Height=NamedRerouteUsage())
  n77 = NamedRerouteUsage()
  n78 = TextureSample(UVs=StaticSwitch(True=Rotator(Coordinate=@n76), False=@n76, Value=NamedRerouteUsage()), Tex=@n77)
  n79 = Multiply(A=Multiply(B=NamedRerouteUsage()) {constA=0.02}, B=@n78.RGB)
  n80 = NamedRerouteUsage()
  n81 = NamedRerouteUsage()
  n82 = TextureSample(UVs=StaticSwitch(True=Rotator(), False=BumpOffset(Coordinate=@n76, Height=@n81), Value=NamedRerouteUsage()), Tex=@n77)
  n83 = Multiply(A=Multiply(B=NamedRerouteUsage()) {constA=0.01}, B=@n82.RGB)
  n84 = NamedRerouteUsage()
  n85 = Add(
   A=Add(A=StaticSwitch(True=@n70, False=@n69.RGB, Value=@n71), B=StaticSwitch(True=@n74, False=@n73.RGB, Value=@n75))
   B=Add(A=StaticSwitch(True=@n79, False=@n78.RGB, Value=@n80), B=StaticSwitch(True=@n83, False=@n82.RGB, Value=@n84))
  )
  NamedRerouteDeclaration(None=Add(A=@n47, B=Add(A=@n66, B=@n85))) {name=Out_EpicQuality, nodeColor=<rgba 1 0 0.6328 1>}
  NamedRerouteDeclaration(None=@n47) {name=Out_HighQuality, nodeColor=<rgba 1 0 0.6328 1>}
  NamedRerouteDeclaration(None=@n47) {name=Out_MediumQuality, nodeColor=<rgba 1 0 0.6328 1>}
  NamedRerouteDeclaration(None=@n28) {name=Out_LowQuality, nodeColor=<rgba 1 0 0.6328 1>}
  n86 = FunctionInput() {bUsePreviewValueAsDefault=true, inputName=Offset, inputType=FunctionInput_Scalar, previewValue=<0.444 0 0 0>}
  NamedRerouteDeclaration(None=@n86) {name=HighOffset, nodeColor=<rgba 1 0.3047 0 1>}
  n87 = FunctionInput() {bUsePreviewValueAsDefault=true, inputName=GradientDarkness, inputType=FunctionInput_Scalar, previewValue=<0.05 0 0 0>}
  NamedRerouteDeclaration(None=@n87) {name=GradientDarkness, nodeColor=<rgba 1 0.0234 0 1>}
  NamedRerouteDeclaration(None=NamedRerouteUsage()) {name=LowOffset, nodeColor=<rgba 1 0.2578 0 1>}
  NamedRerouteDeclaration(None=NamedRerouteUsage()) {name=MediumOffset, nodeColor=<rgba 1 0.2578 0 1>}
  n88 = FunctionInput() {bUsePreviewValueAsDefault=true, inputName=bGradientFalloff, inputType=FunctionInput_StaticBool, previewValue=<1 0 0 0>}
  NamedRerouteDeclaration(None=@n88) {name=bGradientFalloff, nodeColor=<rgba 0 0.2969 1 1>}
  n89 = NamedRerouteUsage()
  NamedRerouteDeclaration(None=QualitySwitch(Default=@n89, Low=NamedRerouteUsage(), High=@n89, Medium=NamedRerouteUsage(), Epic=@n89)) {name=Offset, nodeColor=<rgba 1 0.2578 0 1>}
  n90 = FunctionInput() {bUsePreviewValueAsDefault=true, inputName=bRotateUVs, inputType=FunctionInput_StaticBool, previewValue=<0 0 0 0>}
  NamedRerouteDeclaration(None=@n90) {name=bRotateUVs, nodeColor=<rgba 0 0.0625 1 1>}
  n91 = Constant() {r=1}
  n92 = Constant() {desc=(3/4.5), r=0.6667}
  n93 = Constant() {desc=(7/8), r=0.875}
  NamedRerouteDeclaration(None=Multiply(A=QualitySwitch(Default=@n91, Low=@n92, High=@n91, Medium=@n93, Epic=@n91), B=NamedRerouteUsage())) {name=Quality_15, nodeColor=<rgba 0.6172 1 0 1>}
  n94 = Constant() {r=1}
  n95 = Constant() {desc=(2/4.5), r=0.4444}
  n96 = Constant() {desc=(6/8), r=0.75}
  NamedRerouteDeclaration(None=Multiply(A=QualitySwitch(Default=@n94, Low=@n95, High=@n94, Medium=@n96, Epic=@n94), B=NamedRerouteUsage())) {name=Quality_14, nodeColor=<rgba 0.6172 1 0 1>}
  n97 = Constant() {r=1}
  n98 = Constant() {desc=(1/4.5), r=0.2222}
  n99 = Constant() {desc=(5/8), r=0.625}
  NamedRerouteDeclaration(None=Multiply(A=QualitySwitch(Default=@n97, Low=@n98, High=@n97, Medium=@n99, Epic=@n97), B=NamedRerouteUsage())) {name=Quality_13, nodeColor=<rgba 0.6172 1 0 1>}
  n100 = Constant() {r=1}
  n101 = Constant() {desc=(4/8), r=0.5}
  NamedRerouteDeclaration(None=Multiply(A=QualitySwitch(Default=@n100, Low=@n100, High=@n100, Medium=@n101, Epic=@n100), B=NamedRerouteUsage())) {name=Quality_12, nodeColor=<rgba 0.6172 1 0 1>}
 ]
}
